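# M_GooseItemBG is a translucent UI material that draws an item background for the Goose playlist.
# A base color plus an exponential radial gradient is blended with an outline color through an inverted mask from center-scaled UVs.
# A clamped shine texture, with rotatable UVs shifted equally on both axes by ShineOffset's red channel, is color-dodged over that base.
# WhiteOut fades the shine layer toward solid white before the dodge.
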Material /Game/Athena/Playlists/Goose/UI/Material/M_GooseItemBG
graph {
 properties {blendMode=BLEND_Translucent, materialDomain=MD_UI, pixelDepthOffsetMode=PDOM_Legacy, usedShadingModels=Default Lit}
 nodes [
  n1 = TextureCoordinate()
  n2 = ScalarParameter() {parameterName=ShineRotation}
  n3 = VectorParameter() {parameterName=ShineOffset, defaultValue=<rgba 0 0 0 1>}
  n4 = TextureSample(UVs=Add(
    A=MaterialFunctionCall(UVs=@n1, "Rotation Angle (0-1)"=@n2) {materialFunction=&CustomRotator}.Rotated Values
    B=AppendVector(A=@n3.R, B=@n3.R)
   )) {
   samplerSource=SSM_Clamp_WorldGroupSettings
   samplerType=SAMPLERTYPE_LinearColor
   texture=/Game/UI/Foundation/Textures/Pattern/T-UI-Shine-Anim
  }
  n5 = ScalarParameter() {parameterName=WhiteOut}
  n6 = VectorParameter() {parameterName=BGColor, defaultValue=<rgba 0.1094 0.1094 0.1094 1>}
  n7 = VectorParameter() {parameterName=Offset, defaultValue=<rgba 0.5 0.8 0 1>}
  n8 = ScalarParameter() {defaultValue=1, parameterName=Radius}
  n9 = ScalarParameter() {defaultValue=2, parameterName=Density}
  n10 = Multiply(
   A=VectorParameter() {parameterName=GradientColor, defaultValue=<rgba 0.8021 0.8021 0.8021 1>}.RGB
   B=Multiply(
    A=MaterialFunctionCall(CenterPosition=@n7.RGB, Radius=@n8, Density=@n9) {materialFunction=&RadialGradientExponential}.RadialGradientExponential
    B=ScalarParameter() {defaultValue=0.6, parameterName=RadialOpacity}
   )
  )
  n11 = MaterialFunctionCall(Texture Scale=ScalarParameter() {defaultValue=0.9, parameterName=OutlineSize}) {materialFunction=&ScaleUVsByCenter}
  MaterialFunctionCall!(
   Blend=LinearInterpolate(A=@n4.RGB, B=Constant() {r=1}, Alpha=Clamp(None=@n5))
   Base=LinearInterpolate(
    A=Add(A=@n6.RGB, B=@n10)
    B=VectorParameter() {parameterName=OutlineColor, defaultValue=<rgba 1 1 1 1>}.RGB
    Alpha=OneMinus(None=@n11.0-1 mask)
   )
  ) {materialFunction=/Engine/Functions/Engine_MaterialFunctions03/Blends/Blend_ColorDodge}
 ]
}
